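; Art Deco hotel sign building Blueprint whose construction script tints the FogCard component.
; It splits a fog card color-and-opacity variable into RGBA and writes it as a Vector4 into the card's custom primitive data at index 28.
; The write runs only where cosmetic events can execute.
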(fn ConstructionScript
  (bind (_fogcardcolorandopacity_r _fogcardcolorandopacity_g _fogcardcolorandopacity_b _fogcardcolorandopacity_a) (Math|Color|BreakColor (Variables|GetFogCardColorAndOpacity_R)))
  (Networking|CanExecuteCosmeticEvents
    (:True
      (Rendering|Material|SetCustomPrimitiveDataVector4 (Variables|Default|GetFogCard) 28 (Math|Vector4|MakeVector4 _fogcardcolorandopacity_r _fogcardcolorandopacity_g _fogcardcolorandopacity_b _fogcardcolorandopacity_a)))))
